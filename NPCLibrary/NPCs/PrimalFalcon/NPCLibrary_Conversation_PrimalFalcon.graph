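# Conversation database for the PrimalFalcon NPC, defining the dialogue that opens when a player talks to it.
# The entry point leads to one speech node with a default line about gods and contextual overrides.
# Overrides are picked by curve-table priorities, covering the player's first-ever conversation and reactions to specific worn character outfits.
# The speech then branches through link nodes to the shared NPC library quests and basic services conversations.
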ConversationDatabase /NPCLibrary/NPCs/PrimalFalcon/NPCLibrary_Conversation_PrimalFalcon
entry_points [{tag=Conversation.Entry.PrimalFalcon, destinations [#1]}]
nodes [
 ConversationEntryPointNode_7 {
  id=#1
  class=ConversationEntryPointNode
  outputs [#2]
  properties {entryTag {tagName=Conversation.Entry.PrimalFalcon}}
 }
 Speech_C_4 {
  id=#2
  class=Speech_C
  outputs [#3, #4]
  properties {
   nodeName=""
   generalConfig {
    defaultMessage=You shouldn't fear the gods. They should fear you.
    contextualMessages [
     {
      message=You're lucky you're not a god. I might have hurt you.
      contextRequirements [{requirement {class=FortControllerRequirement_IsFirstEverConversationWithNPC}}]
      priority {curve {rowName=Tandem.Speech.Priority.FirstEverInteraction}}
     }
     {
      message="You, above all, should understand why I hunt."
      contextRequirements [{requirement {
         class=FortControllerRequirement_HasCID
         properties {allowedCIds [/BRCosmetics/Athena/Items/Cosmetics/Characters/Character_SeleneCobra_Stride]}
        }}]
      priority {curve {rowName=Tandem.Speech.Priority.OutfitReaction}}
     }
     {
      message="When this is over, I'll wield your power as a trophy."
      contextRequirements [{requirement {
         class=FortControllerRequirement_HasCID
         properties {allowedCIds [&Character_StunningMountain_Armor]}
        }}]
      priority {curve {rowName=Tandem.Speech.Priority.OutfitReaction}}
     }
     {
      message="I wonder, have your chains ever held you?"
      contextRequirements [{requirement {
         class=FortControllerRequirement_HasCID
         properties {allowedCIds [/BRCosmetics/Athena/Items/Cosmetics/Characters/Character_DarkStance_Inferno]}
        }}]
      priority {curve {rowName=Tandem.Speech.Priority.OutfitReaction}}
     }
     {
      message=Hearts were meant to be broken. I am simply the hammer.
      contextRequirements [{requirement {class=FortControllerRequirement_HasCID, properties {allowedCIds [/BRCosmetics/Athena/Items/Cosmetics/Characters/Character_LiftingAura_Wave]}}}]
      priority {curve {rowName=Tandem.Speech.Priority.OutfitReaction}}
     }
     {
      message=The sea's tempests don't hold a candle to me.
      contextRequirements [{requirement {class=FortControllerRequirement_HasCID, properties {allowedCIds [/BRCosmetics/Athena/Items/Cosmetics/Characters/Character_TideKing_Regal]}}}]
      priority {curve {rowName=Tandem.Speech.Priority.OutfitReaction}}
     }
    ]
   }
  }
 }
 ConversationLinkNode_0 {id=#4, class=ConversationLinkNode, properties {remoteEntryTag {tagName=Conversation.Entry.BasicServices}}}
 ConversationLinkNode_1 {id=#3, class=ConversationLinkNode, properties {remoteEntryTag {tagName=Conversation.Entry.NPCLibrary.Quests}}}
]
